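# Sound cue for the stereo platform drop sound effect in the SparksMaps Battle Stage audio.
# A single wave player feeds the Platform_Drop_Stereo wave straight to the output with boosted volume, routed through the Fort_SFX_InGame sound class.
# A concurrency override limits it to one playing instance at a time.
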SoundCue /SparksMaps/Audio/BattleStage/Platform_Drop_Stereo_Cue
root=WavePlayer_0
nodes [{id=WavePlayer_0, properties {soundWaveAssetPtr=&/Platform_Drop_Stereo}}]
properties {
 bOverrideConcurrency=true
 maxDistance=2097152
 soundClassObject=&Fort_SFX_InGame
 volumeMultiplier=2
 concurrencyOverrides {maxCount=1}
}
