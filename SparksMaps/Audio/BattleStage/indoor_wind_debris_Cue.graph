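# Looping ambient Sound Cue for indoor wind and debris audio in the BattleStage map audio set.
# A Mixer root layers two looping Wave Players, an indoor wind debris wave and a window debris loop, with the window layer weighted louder.
# The cue lowers the overall volume and sets no attenuation, leaving its max distance at world scale.
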SoundCue /SparksMaps/Audio/BattleStage/indoor_wind_debris_Cue
root=Mixer_0
nodes [
 {id=Mixer_0, children [WavePlayer_0,WavePlayer_1], properties {inputVolume [1,2]}}
 {id=WavePlayer_0, properties {bLooping=true, soundWaveAssetPtr=&/indoor_wind_debris}}
 {id=WavePlayer_1, properties {bLooping=true, soundWaveAssetPtr=&/HMX_BS_GR_window_debris_loop}}
]
properties {maxDistance=2097152, volumeMultiplier=0.5}
