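# Sound cue that plays the "opponents streak broken" announcement sound in the BattleStage audio set.
# A single wave player feeds the root directly, with no mixing, randomization or modulation nodes.
# Its very large max distance suggests it is meant to be heard regardless of listener position.
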SoundCue /SparksMaps/Audio/BattleStage/opponents_streak_broken_Cue
root=WavePlayer_0
nodes [{id=WavePlayer_0, properties {soundWaveAssetPtr=&/opponents_streak_broken}}]
properties {maxDistance=2097152}
