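# Sound cue in the Sparks Battle Stage audio set that plays the platform_lower_revision sound, named for a stage platform lowering.
# A single wave player node feeds the root directly, with no randomization or modulation.
# The cue uses the Fort_SFX_InGame sound class and sets a very large max distance.
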SoundCue /SparksMaps/Audio/BattleStage/platform_lower_revision_Cue
root=WavePlayer_0
nodes [{id=WavePlayer_0, properties {soundWaveAssetPtr=&/platform_lower_revision}}]
properties {maxDistance=2097152, soundClassObject=&Fort_SFX_InGame}
